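; Tapas SkyStrike projectile: a falling hammer that damages what it passes through and explodes on reaching ground or water.
; A looping timer traces ahead by velocity for ground, while the server capsule-overlaps that span, applying environmental, player or knockback GameplayEffects and ignoring actors already hit.
; Impact stops movement, explodes, shows a sound indicator and runs a final overlap that always damages non-teammates.
; The hammer then snaps to the floor, lingering briefly before removal.
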(event EventBeginPlay
  ;[1 1 1] BeginPlay->Start Timer to have the hammer break buildings as it falls.
  (Utilities|Array|Add (Variables|Default|GetOverlapBuildingActorsToIgnore) (Variables|Actor|GetInstigator))
  (CallFunction|DestroyonTick) ; Need to do one right away
  (bind _returnvalue (Utilities|Time|SetTimerbyEvent (delegate Custom|DestroyOnTick) (Variables|Default|GetDestroyBuildsTick) true))
  (Variables|Default|SetDestroyBuildingsTimer _returnvalue))

(event Custom|ApplyHammerEffects
  (params (Array :type Array%20of%20Actor%20Object%20References) (AlwaysDamagePlayers :type Boolean))
  (bind _instigator (Variables|Actor|GetInstigator))
  (bind _returnvalue (Ability|GetAbilitySystemComponent _instigator))
  (bind _instigator_2 (Variables|Actor|GetInstigator))
  (bind _returnvalue_2 (Ability|GetAbilitySystemComponent _instigator_2))
  ;[1 1 1] BeginPlay->Start Timer to have the hammer break buildings as it falls.
  (for _array_element Array
    (Utilities|Array|Add (Variables|Default|GetOverlapBuildingActorsToIgnore) _array_element)
    (Utilities|Casting|CastToPawn _array_element
      (:then
        _)
      (:CastFailed
        (Utilities|Casting|CastToFortWeakPoint _array_element
          (:then
            _)
          (:CastFailed
            (if (Variables|Actor|GetbCanBeDamaged _array_element)
              (GameplayEffects|ApplyGameplayEffectToTarget :self _returnvalue :GameplayEffectClass (Variables|Default|GetGE_EnvironmentalDamage) :Target (Ability|GetAbilitySystemComponent _array_element) :Level 0.0)
              (GameplayEffects|ApplyGameplayEffectToTarget :self _returnvalue :GameplayEffectClass (Variables|Default|GetGE_Knockback) :Target (Ability|GetAbilitySystemComponent _array_element) :Level 0.0)))) ; FortWeakPoint is only used as an attachment on AI Pawns for now so treat it like a pawn. Would prefer to check if this is a child actor of a pawn but that was failing in real games
      ))
    (if (and (or AlwaysDamagePlayers (Class|AthenaFunctionLibrary|RowtoBool (Variables|Default|GetDamagePlayersWhileFalling))) (not (Team|OnSameTeam (Variables|Actor|GetInstigator) _array_element)))
      (if (Utilities|IsValid _array_element :!class /Script/Engine.KismetSystemLibrary)
        (GameplayEffects|ApplyGameplayEffectToTarget :self _returnvalue_2 :GameplayEffectClass (Variables|Default|GetGE_PlayerDamage) :Target (Ability|GetAbilitySystemComponent _array_element) :Level 0.0)
        (GameplayEffects|ApplyGameplayEffectToTarget :self _returnvalue_2 :GameplayEffectClass (Variables|Default|GetGE_Knockback) :Target (Ability|GetAbilitySystemComponent _array_element) :Level 0.0)))))

(event Custom|DestroyOnTick
  ;[1 1 1] BeginPlay->Start Timer to have the hammer break buildings as it falls.
  (bind (_out_hit_blocking_hit _out_hit_distance _out_hit_impact_point) (Utilities|FindGround (Transformation|GetActorLocation) (Math|Float|float*float (ScalableFloat|GetValueatLevel (Variables|Default|GetProjectileVelocity)) (Variables|Default|GetDestroyBuildsTick) (Variables|Default|GetDestroyBuildsBuffer))))
  (if _out_hit_blocking_hit
    (CallFunction|HitGround _out_hit_impact_point))
  (Networking|SwitchHasAuthority
    (:Authority
      (bind _outactors (Collision|CapsuleOverlapActors (Math|Vector|vector-vector (Transformation|GetActorLocation) (Math|Vector|MakeVector 0.0 0.0 (* _out_hit_distance 0.5))) (ScalableFloat|GetValueatLevel (Variables|Default|GetOverlapBuildingRadius)) (+ (* _out_hit_distance 0.5) (ScalableFloat|GetValueatLevel (Variables|Default|GetOverlapBuildingRadius))) (Variables|Default|GetOverlapBuildingObjectType) 0 (Variables|Default|GetOverlapBuildingActorsToIgnore)))
      (CallFunction|ApplyHammerEffects _outactors))))

(event OnEnterWaterEvent(WaterInteractionComponent)
  (params (WaterBody :type Fort%20Water%20Body%20Actor%20Object%20Reference) (WaterInteractionComponent :type Fort%20Water%20Interaction%20Component%20Object%20Reference) (bIsFirstBody :type Boolean))
  (bind _returnvalue (Transformation|GetActorLocation))
  (bind _hitwater (Variables|Default|GetHitWater))
  ;[1 1 1] EventHit-> Do explosion and turn off collision
  (if (not (Variables|Default|GetStopped))
    (Variables|Default|SetStopped true)
    (Variables|Default|SetHitWater true)
    (Audio|PlaySoundatLocation (Variables|Sound|GetWater%20Impact%20Sound) _returnvalue)
    (Projectile|SetVelocity (Variables|Movement|GetProjectileMovementComponent) (0.0 0.0 0.0))
    (Collision|SetActorEnableCollision false)
    (Delivery|DoExplosion)
    (bind _outwatersurfaceinfo (WaterBody|GetWaterSurfaceInfoStruct WaterBody _returnvalue true false false))
    (bind _outwatersurfaceinfo_waterplanelocation (Utilities|Struct|BreakWaterSurfaceInfo _outwatersurfaceinfo))
    (Fortnite|ShowSoundIndicatorLocalforTeam :TrackedActor self :Location (Math|Vector|SelectVector _outwatersurfaceinfo_waterplanelocation (stale Custom|HitGround GroundPos) _hitwater) :MaxAudibleDistance 1e+04 :IndicatorType "Gunshot" :Team (Team|GetActorTeam self) :Affiliations (Utilities|Array|MakeArray "Hostile" "Neutral") :Instigator (Variables|Actor|GetInstigator) :Tint "(R=0.000000,G=0.000000,B=0.000000,A=0.000000)" :AttenuationMaxRangeMultiplier 1.0 :AttenuationIntensityMultiplier 1.0)
    (Networking|SwitchHasAuthority
      (:Authority
        (bind _outactors (Collision|CapsuleOverlapActors (Math|Vector|SelectVector _outwatersurfaceinfo_waterplanelocation (stale Custom|HitGround GroundPos) _hitwater) 375.0 5e+02 (Variables|Default|GetOverlapBuildingObjectType) 0 (Variables|Default|GetOverlapBuildingActorsToIgnore)))
        (CallFunction|ApplyHammerEffects _outactors true)
        (Utilities|Time|ClearandInvalidateTimerbyHandle (Variables|Default|GetDestroyBuildingsTimer))))))

(event Custom|HitGround
  (params (GroundPos :type Vector))
  (bind _hitwater (Variables|Default|GetHitWater))
  ;[1 1 1] EventHit-> Do explosion and turn off collision
  (if (not (Variables|Default|GetStopped))
    (Variables|Default|SetStopped true)
    (Transformation|SetActorLocation GroundPos false true)
    (Projectile|SetVelocity (Variables|Movement|GetProjectileMovementComponent) (0.0 0.0 0.0))
    (Collision|SetActorEnableCollision false)
    (Delivery|DoExplosion)
    (bind _outwatersurfaceinfo (WaterBody|GetWaterSurfaceInfoStruct (stale OnEnterWaterEvent(WaterInteractionComponent) WaterBody) (Transformation|GetActorLocation) true false false))
    (bind _outwatersurfaceinfo_waterplanelocation (Utilities|Struct|BreakWaterSurfaceInfo _outwatersurfaceinfo))
    (Fortnite|ShowSoundIndicatorLocalforTeam :TrackedActor self :Location (Math|Vector|SelectVector _outwatersurfaceinfo_waterplanelocation GroundPos _hitwater) :MaxAudibleDistance 1e+04 :IndicatorType "Gunshot" :Team (Team|GetActorTeam self) :Affiliations (Utilities|Array|MakeArray "Hostile" "Neutral") :Instigator (Variables|Actor|GetInstigator) :Tint "(R=0.000000,G=0.000000,B=0.000000,A=0.000000)" :AttenuationMaxRangeMultiplier 1.0 :AttenuationIntensityMultiplier 1.0)
    (Networking|SwitchHasAuthority
      (:Authority
        (bind _outactors (Collision|CapsuleOverlapActors (Math|Vector|SelectVector _outwatersurfaceinfo_waterplanelocation GroundPos _hitwater) 375.0 5e+02 (Variables|Default|GetOverlapBuildingObjectType) 0 (Variables|Default|GetOverlapBuildingActorsToIgnore)))
        (CallFunction|ApplyHammerEffects _outactors true)
        (Utilities|Time|ClearandInvalidateTimerbyHandle (Variables|Default|GetDestroyBuildingsTimer))))))

(event EventOnExploded
  (params (HitActors :type Array%20of%20Actor%20Object%20References) (HitResults :type Array%20of%20Hit%20Result%20Structures))
  ;[1 1 1] OnExploded-> Remove feedback, set life span.  Does not kill on explode so we get to see the hammer before it goes
  (Utilities|FlowControl|Sequence
    (:then_0
      (Components|Activation|Deactivate (Variables|Default|GetInAirSound))
      (Networking|ForceNetUpdate)
      (Actor|SetLifeSpan 2.0)
      (Utilities|Time|SetTimerbyEvent (delegate Custom|SnapToGround) 0.2) ; Wait for hammer to reach the ground
    )
    (:then_1
      (Utilities|FlowControl|Delay 1.9)
      (Effects|SpawnEmitteratLocation "/HighTower/Items/Tapas/SkyStrike/PROTO/P_HighTower_Tapas_Hammer_Del.P_HighTower_Tapas_Hammer_Del" (Transformation|GetActorLocation)))))

(event Custom|SnapToGround
  ;[1 1 1] OnExploded-> Remove feedback, set life span.  Does not kill on explode so we get to see the hammer before it goes
  (bind _location (Utilities|SnaptoFloor (Transformation|GetWorldLocation (Variables|Default|GetHighTower_Tapas_Axe))))
  (Transformation|SetWorldLocation (Variables|Default|GetHighTower_Tapas_Axe) _location false true))

(fn ConstructionScript)
