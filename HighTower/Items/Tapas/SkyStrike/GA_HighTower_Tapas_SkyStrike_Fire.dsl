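; Sky Strike (Tapas) fire ability: aims with a ground indicator and drops a hammer projectile from the sky.
; A non-authority timer moves the indicator to a camera trace snapped downward.
; On target, FindHammerPath traces from above to set the spawn point and plays a cloud cue; a montage notify launches the projectile.
; Optional hover restores gravity over time by reapplying a SetByCaller effect; anim layers and indicator clear on end.
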(event Ability|EventActivateAbility
  (bind _playerpawn (Variables|Default|GetPlayerPawn))
  ;[1 1 1] Spawning and updating Aim Indicator (cylinder)
  (Variables|PowerHover|SetHoverOnActivate (Class|AthenaFunctionLibrary|RowtoBool (Variables|DesignVariables|GetAllowHover)))
  (Variables|PowerHover|SetHoverGravityMultiplier (ScalableFloat|GetValueatLevel (Variables|DesignVariables|GetHoverGravity)))
  (|CallParentFunction)
  (Utilities|FlowControl|Sequence
    (:then_0
      (bind _outactors (Actor|GetAllActorsOfClass "/Script/Engine.WorldSettings"))
      (Variables|HammerTrace|SetWorldMinZ (Variables|/Script/Engine.WorldSettings|GetKillZ (Utilities|Array|Get(acopy) _outactors)))
      (Utilities|IsValid (Variables|Indicator|GetAimIndicator)
        (:Is%20Valid
          (Actor|DestroyActor (Variables|Indicator|GetAimIndicator)))))
    (:then_1
      (Ability|HasAuthority
        (:False
          (bind _returnvalue (Game|SpawnActorfromClass "/HighTower/Items/Tapas/SkyStrike/BP_SkyStrike_AimIndicator.BP_SkyStrike_AimIndicator_C" (struct :Location "0, 0, 0" :Rotation "0, 0, 0" :Scale "1.000000,1.000000,1.000000") "Undefined" "OverrideRootScale" 0 0 0))
          (Variables|Indicator|SetAimIndicator _returnvalue)
          (bind _returnvalue_2 (Utilities|Time|SetTimerbyEvent (delegate Custom|SetIndicatorPos) 0.016 true))
          (Variables|Indicator|SetSetIndicatorPosTimer _returnvalue_2))))
    (:then_2
      (Utilities|IsServer
        (:true
          (Animation|MulticastSetWeaponOverrideAnimSet _playerpawn "/HighTower/Animation/Tapas/HighTower_Tapas_AnimSet.HighTower_Tapas_AnimSet")
          (Animation|MulticastLinkAnimLayer _playerpawn "/HighTower/Animation/Tapas/HighTower_Tapas_LayerAnimBP.HighTower_Tapas_LayerAnimBP_C"))
        (:false
          _))
      (Ability|AddGameplayCueToOwner (Variables|Data|GetAimingGC)))))

(event Custom|DelayThenStartBringingGravityBack
  ;[1 1 1] Bring back gravity over time
  (Utilities|Time|SetTimerbyEvent (delegate Custom|StartBringingGravityBack) (Variables|Default|GetDelayBeforeGravityStartsComingBack)))

(event Custom|SetIndicatorPos
  ;[1 1 1] Spawning and updating Aim Indicator (cylinder)
  (bind _outhit_location (CallFunction|GetAimedatPoint))
  (Transformation|SetActorLocation (Variables|Indicator|GetAimIndicator) _outhit_location false true))

(event Custom|StartBringingGravityBack
  ;[1 1 1] Bring back gravity over time
  (bind _returnvalue (Utilities|Time|SetTimerbyEvent (delegate Custom|IncreaseGravity) (Variables|Default|GetGravityUpdateTime) true))
  (Variables|Default|SetGravityRejuvinationTimer _returnvalue))

(event Custom|IncreaseGravity
  ;[1 1 1] Bring back gravity over time
  (Variables|Default|SetCurrentGravity (Utilities|CalculateNewGravity))
  (Ability|RemoveGameplayEffectFromOwnerWithHandle (Variables|PowerHover|GetGEH_Hover))
  (bind _returnvalue (Ability|GameplayEffect|AssignTagSetbyCallerMagnitude (Ability|MakeOutgoingGameplayEffectSpec (Variables|PowerHover|GetGE_HoverGravity) 0.0) "(TagName=\"SetByCaller.Magnitude\")" (Variables|Default|GetCurrentGravity)))
  (bind _returnvalue_2 (Ability|GameplayEffect|AssignTagSetbyCallerMagnitude _returnvalue "(TagName=\"SetByCaller.MoveSpeed\")" (Variables|PowerHover|GetHoverAirControlAmount)))
  (bind _returnvalue_3 (Ability|ApplyGameplayEffectSpecToOwner _returnvalue_2))
  (Variables|PowerHover|SetGEH_Hover _returnvalue_3))

(event Ability|EventOnEndAbility
  (params (bWasCancelled :type Boolean))
  (bind _playerpawn (Variables|Default|GetPlayerPawn))
  ;[1 1 1] End ability when montage ends
  (|CallParentFunction false)
  (Input|ResetInputRotationScale (Fortnite|GetFortPlayerControllerfromActor (Variables|Default|GetPlayerPawn)))
  (Actor|DestroyActor (Variables|Indicator|GetAimIndicator))
  (Utilities|IsServer
    (:true
      (Animation|MulticastClearWeaponOverrideAnimSet _playerpawn)
      (Animation|MulticastUnlinkAnimLayer _playerpawn "/HighTower/Animation/Tapas/HighTower_Tapas_LayerAnimBP.HighTower_Tapas_LayerAnimBP_C"))
    (:false
      _))
  (CallFunction|SetHover false)
  (Actor|DestroyActor (Variables|Indicator|GetAimIndicator)))

(event EventTarget
  (bind _aimindicator (Variables|Indicator|GetAimIndicator))
  ;[1 1 1] On Target, break builds / knock back players.
  (|CallParentFunction)
  (Actor|DestroyActor _aimindicator)
  (Rendering|SetActorHiddenInGame _aimindicator true)
  (Ability|RemoveGameplayCueFromOwner (Variables|Data|GetAimingGC))
  (CallFunction|FindHammerPath)
  (Utilities|Time|ClearandInvalidateTimerbyHandle (Variables|Indicator|GetSetIndicatorPosTimer))
  ;[1 1 1] Play GC when player pulls hand down
  (Utilities|FlowControl|Delay 0.25)
  (CallFunction|PlayBurstGC (Variables|Default|GetGC_HandCastEffects) true (Ability|GameplayCue|MakeGameplayCueParameters :NormalizedMagnitude 0.0 :RawMagnitude 0.0 :Location (Transformation|GetActorLocation (Ability|GetOwningActorfromActorInfo)) :Normal "0,0,0" :Instigator (Ability|GetOwningActorfromActorInfo) :EffectCauser (Ability|GetOwningActorfromActorInfo) :SourceObject (Ability|GetOwningActorfromActorInfo) :GameplayEffectLevel 1 :AbilityLevel 1 :bReplicateLocationWhenUsingMinimalRepProxy false)))

(event EventCastMontageNotifyStart
  (params (NotifyTag :type Gameplay%20Tag%20Structure))
  ;[1 1 1] Spawn the hammer when the projectile notify comes in
  (if (GameplayTags|Equal(GameplayTag) NotifyTag (Variables|Data|GetEventTagSpawnProjectile))
    (Projectile|SpawnProjectile (Variables|Data|GetProjToSpawn) (Variables|Default|GetPlayerPawn) (Variables|HammerTrace|GetHammerTraceStart) "0, 0, 0" (Math|Rotator|MakeRotfromX (Math|Vector|Normalize (- (Variables|HammerTrace|GetHammerTraceEnd) (Variables|HammerTrace|GetHammerTraceStart)))) (Ability|MakeGameplayEffectContainerSpec (Variables|Data|GetTag_SpawnProjectiles) 0) (Ability|MakeGameplayEffectContainerSpec (Variables|Data|GetTag_SpawnProjectiles) 0) false 0 (ScalableFloat|GetValueatLevel (Variables|DesignVariables|GetProjDownwardSpeed)) (Variables|DesignVariables|GetGravityScale))
    (Variables|Data|SetSpawnedProjectileAlready true)
    (else
      (|CallParentFunction NotifyTag))))

(fn FindHammerPath
  ;[1 1 1] FindHammerPath-> Trace from 20k units above you.  If we hit something, spawn the hammer above that.  If we dont, spawn it 5k units above you.
  (bind _outhit_location (CallFunction|GetAimedatPoint))
  (bind _output_get (Variables|HammerTrace|SetHammerTraceEnd _outhit_location))
  (bind (_outhit _returnvalue) (Collision|LineTraceByChannel (+ _output_get (Variables|HammerTrace|GetTraceOffset)) _output_get "TraceTypeQuery7"))
  (bind _output_get_2 (Variables|HammerTrace|SetHammerTraceStart (+ (select _returnvalue (.Location (Collision|BreakHitResult _outhit)) _output_get) (Variables|DesignVariables|GetSpawnOffset))))
  ;[1 1 1] Sky Cloud GC
  (CallFunction|PlayBurstGC (Variables|Data|GetGC_CloudFormation) true (Ability|GameplayCue|MakeGameplayCueParameters :NormalizedMagnitude 0.0 :RawMagnitude 0.0 :Location _output_get_2 :Normal (Variables|HammerTrace|GetHammerTraceEnd) :GameplayEffectLevel 1 :AbilityLevel 1 :bReplicateLocationWhenUsingMinimalRepProxy false)))

(fn GetAimedAtPoint
  (bind (_returnvalue_owneractor _returnvalue_avataractor) (Utilities|Struct|BreakGameplayAbilityActorInfo (Ability|GetActorInfo)))
  (bind (_returnvalue_owneractor_2 _returnvalue_avataractor_2) (Utilities|Struct|BreakGameplayAbilityActorInfo (Ability|GetActorInfo)))
  (bind (_returnvalue_x _returnvalue_y _returnvalue_z) (Math|Vector|BreakVector (Transformation|GetActorLocation _returnvalue_avataractor_2)))
  (bind _outhit (Collision|LineTraceByProfile (Camera|GetCameraLocation (Variables|/Script/Engine.PlayerController|GetPlayerCameraManager (Fortnite|GetFortPlayerControllerfromActor (Utilities|Casting|CastToFortPlayerPawn _returnvalue_avataractor)))) (+ (Camera|GetCameraLocation (Variables|/Script/Engine.PlayerController|GetPlayerCameraManager (Fortnite|GetFortPlayerControllerfromActor (Utilities|Casting|CastToFortPlayerPawn _returnvalue_avataractor)))) (* (Math|Vector|GetForwardVector (Camera|GetCameraRotation (Variables|/Script/Engine.PlayerController|GetPlayerCameraManager (Fortnite|GetFortPlayerControllerfromActor (Utilities|Casting|CastToFortPlayerPawn _returnvalue_avataractor))))) (ScalableFloat|GetValueatLevel (Variables|DesignVariables|GetMaxRange)))) "FortTrajectoryBlocksWater" false (Utilities|Array|MakeArray _returnvalue_avataractor)))
  (LocalVariables|SetHitSomething (Collision|BreakHitResult _outhit))
  (bind _outhit_2 (Collision|LineTraceByChannel :Start (+ (Math|Vector|SelectVector (.Location (Collision|BreakHitResult _outhit)) (.TraceEnd (Collision|BreakHitResult _outhit)) (Collision|BreakHitResult _outhit)) "0, 0,1000.000000") :End (Math|Vector|MakeVector (.x (Math|Vector|SelectVector (.Location (Collision|BreakHitResult _outhit)) (.TraceEnd (Collision|BreakHitResult _outhit)) (Collision|BreakHitResult _outhit))) (.y (Math|Vector|SelectVector (.Location (Collision|BreakHitResult _outhit)) (.TraceEnd (Collision|BreakHitResult _outhit)) (Collision|BreakHitResult _outhit))) (Math|Float|SelectFloat (- _returnvalue_z 6e+02) (Variables|HammerTrace|GetWorldMinZ) (LocalVariables|GetHitSomething))) :TraceChannel "TraceTypeQuery7" :bTraceComplex false :DrawDebugType "None" :bIgnoreSelf true :TraceColor "(R=1.000000,G=0.000000,B=0.000000,A=1.000000)" :TraceHitColor "(R=0.000000,G=1.000000,B=0.000000,A=1.000000)" :DrawTime 5.0))
  (bind (_outhit_bblockinghit _outhit_binitialoverlap _outhit_time _outhit_distance _outhit_location) (Collision|BreakHitResult _outhit_2))
  (if _outhit_bblockinghit
    (return _outhit_location)
    (else
      (return (Math|Vector|SelectVector (.Location (Collision|BreakHitResult _outhit)) (.TraceEnd (Collision|BreakHitResult _outhit)) (Collision|BreakHitResult _outhit))))))

(fn SetHover
  (params (On :type Boolean))
  (|CallParentFunction On)
  (if On
    (CallFunction|DelayThenStartBringingGravityBack)
    (Variables|Default|SetCurrentGravity (Variables|PowerHover|GetHoverGravityMultiplier))
    (else
      (Utilities|Time|ClearandInvalidateTimerbyHandle (Variables|Default|GetGravityRejuvinationTimer)))))
